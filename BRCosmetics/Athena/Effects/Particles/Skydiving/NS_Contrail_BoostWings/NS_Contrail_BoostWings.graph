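# Niagara skydiving contrail cosmetic (BoostWings) that puts meshes on both forearms with thruster-like flashes, lights, smoke, sparks and ribbons.
# Single persistent mesh particles follow each lower arm and compute a rotation from System.Dive, while flash meshes spawn at the hands.
# System.Dive, from diving-attribute modules, also crossfades smoke layers and scales flash, spark and ribbon spawn rates and light color.
# Two spawn groups split lights, smoke and sparks across User.SkeletalMesh attach points.
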NiagaraSystem /BRCosmetics/Athena/Effects/Particles/Skydiving/NS_Contrail_BoostWings/NS_Contrail_BoostWings
dependencies {
 usedDataInterfaces [
  &NiagaraDataInterfaceCurve
  &NiagaraDataInterfaceArrayFloat3
  &NiagaraDataInterfaceArrayFloat
  &NiagaraDataInterfaceColorCurve
  &NiagaraDataInterfaceVectorCurve
 ]
 usedDynamicInputs []
 usedModules [&SystemState, /Game/Effects/Niagara/Modules/Skydiving/FN_DivingAttributes/v4/FN_DivingAttributes_v4]
}
emitters {
 Flash {
  emitterUpdateScript [
   &EmitterState()
   &SpawnRate(SpawnRate=&Lerp_Float(A=0, B=7, Alpha=$System.Dive), Use Spawn Probability=false) {module=SpawnRate001}
  ]
  particleSpawnScript [
   /Niagara/Modules/Spawn/Initialization/InitializeParticle(
    Lifetime=&UniformRangedFloat(Minimum=0.1, Maximum=0.1)
    Mesh Scale=&RandomRangeVector(Minimum=<0.2 0.3 0.3>, Maximum=<0.468 0.3 0.3>)
    Module.Write Color=true
    Module.Write Lifetime=true
    Module.Write Mass=false
    Module.Write Material Random=true
    Module.Write Position=true
    Module.Write Scale=true
    Module.Write SpriteRotation=false
    Module.Write SpriteSize=false
    Position=&Add_Vector(A=/Niagara/DynamicInputs/Helpers/SimulationPosition(), B=<0 0 0>)
   )
   &FN_SetSpawnGroup_v2(Number of Spawn Groups=2)
   &SkeletalMeshLocation(
    Bone Sampling Mode=::NewEnumerator0
    Inherit Velocity=false
    Interpolate Mesh Sampling=false
    MaskSpawnGroup=false
    Module.Override Randomness=false
    Sampled Position Offset=<0 0 5>
    Skeletal Mesh=DataInterface() {propertyValues {
      DefaultMesh=null
      ExcludeBoneName=root
      OptionalComponentClass=null
      PreviewMesh=&M_Med_Soldier_04_Preview
      SkinningMode=SkinOnTheFly
      SoftSourceActor=null
      SourceMode=Default
      UvSetIndex=0
      WholeMeshLOD=-1
      bExcludeBone=false
      bReadDeformedGeometry=true
      bRequireCurrentFrameData=true
      ComponentTags []
      FilteredBones [hand_l]
      FilteredSockets []
      MeshUserParameter {parameter {name=None, typeDefHandle {registeredTypeIndex=1}, varData []}}
      SamplingRegions []
     }}
   )
   /Game/Effects/Niagara/Modules/Orientation/FN_Fan(
    Cone Angle:NiagaraFloat=&UniformRangedFloat(Minimum=0, Maximum=0)
    Spin:NiagaraFloat=&UniformRangedFloat(Minimum=0, Maximum=6.2)
   )
   &InitialMeshOrientation(Orientation Vector=<0 0 0.25>, Use Orientation Vector=true, Use Rotation Vector=false)
  ]
  particleUpdateScript [
   &ParticleState()
   /SetVariables(Particles.MeshOrientation=&MultiplyQuaternion(
     Quaternion A=$Output.SkeletalMeshLocation.SampledRotation
     Quaternion B=&MultiplyQuaternion(
      Quaternion A=&FindQuatBetweenTwoVectors(From Vector=<0 0 0>, To Vector=<0 0 0>)
      Quaternion B=&MakeQuatFromAxisAngle(Axis=<0 0 0>, AngleInDegrees=0)
     )
    )) {enabled=false, isSetParameters=true, module=SetVariables_#1}
   &Color(
    Color=<rgba 1 0.0833 0 1>
    Scale Color=&VectorFromCurve(VectorCurve=DataInterface() {propertyValues {
       XCurve {keys [<at -0.0031 48.8>,<at 1 1>]}
       YCurve {keys [<at -0.006 20.3>,<at 1 0.8>]}
       ZCurve {keys [<at -0.0011 1.0067>,<at 0.9989 0.28>]}
      }})
   )
   &ScaleMeshSize(Scale Factor=&VectorFromCurve(VectorCurve=DataInterface() {propertyValues {
       XCurve {keys [{arriveTangent=6.3957, interpMode=RCIM_Cubic, leaveTangent=6.3957, tangentMode=RCTM_User, time=0, value=0.47}, <at 0.9748 1.29>]}
       YCurve {keys [{arriveTangent=6.1645, interpMode=RCIM_Cubic, leaveTangent=6.1645, tangentMode=RCTM_User, time=0.0035, value=0.4903}, <at 0.9786 1.29>]}
       ZCurve {keys [
         <at -0.0011 1.0067 RCIM_Cubic>
         {arriveTangent=-1.635, interpMode=RCIM_Cubic, leaveTangent=-1.635, tangentMode=RCTM_User, time=1.0016, value=0.28}
        ]}
      }}))
   &CameraOffset(Camera Offset Amount=-10)
  ]
  rendererClasses [&NiagaraMeshRendererProperties]
  renderers [{rendererClass=&NiagaraMeshRendererProperties, rendererIndex=0}]
 }
 Flash001 {
  emitterUpdateScript [
   &EmitterState(Life Cycle Mode=::NewEnumerator1, Loop Duration=1, UseLoopDelay=false)
   &SpawnRate(SpawnRate=&Lerp_Float(A=0, B=7, Alpha=$System.Dive), Use Spawn Probability=false) {module=SpawnRate001}
  ]
  particleSpawnScript [
   &InitializeParticle(
    Lifetime=&UniformRangedFloat(Minimum=0.05, Maximum=0.06)
    Mesh Scale=&RandomRangeVector(Minimum=<0.2 0.3 0.3>, Maximum=<0.468 0.3 0.3>)
    Module.Write Color=true
    Module.Write Lifetime=true
    Module.Write Mass=false
    Module.Write Material Random=true
    Module.Write Position=true
    Module.Write Scale=true
    Module.Write SpriteRotation=false
    Module.Write SpriteSize=false
    Position=&Add_Vector(A=&SimulationPosition(), B=<0 0 0>)
   )
   &FN_SetSpawnGroup_v2(Number of Spawn Groups=2)
   &SkeletalMeshLocation(
    Bone Sampling Mode=::NewEnumerator0
    Inherit Velocity=false
    Interpolate Mesh Sampling=false
    MaskSpawnGroup=false
    Module.Override Randomness=false
    Sampled Position Offset=<0 0 5>
    Skeletal Mesh=DataInterface() {propertyValues {
      DefaultMesh=null
      ExcludeBoneName=root
      OptionalComponentClass=null
      PreviewMesh=&M_Med_Soldier_04_Preview
      SkinningMode=SkinOnTheFly
      SoftSourceActor=null
      SourceMode=Default
      UvSetIndex=0
      WholeMeshLOD=-1
      bExcludeBone=false
      bReadDeformedGeometry=true
      bRequireCurrentFrameData=true
      ComponentTags []
      FilteredBones [hand_r]
      FilteredSockets []
      MeshUserParameter {parameter {name=None, typeDefHandle {registeredTypeIndex=1}, varData []}}
      SamplingRegions []
     }}
   )
   &FN_Fan(
    Cone Angle:NiagaraFloat=&UniformRangedFloat(Minimum=0, Maximum=0)
    Spin:NiagaraFloat=&UniformRangedFloat(Minimum=0, Maximum=6.2)
   )
   &InitialMeshOrientation(Orientation Vector=<0 0 0.25>, Use Orientation Vector=true, Use Rotation Vector=false)
  ]
  particleUpdateScript [
   &ParticleState()
   /SetVariables(Particles.MeshOrientation=&MultiplyQuaternion(
     Quaternion A=$Output.SkeletalMeshLocation.SampledRotation
     Quaternion B=&MultiplyQuaternion(
      Quaternion A=&FindQuatBetweenTwoVectors(From Vector=<0 0.2 0.2>, To Vector=<0 0 0>)
      Quaternion B=&MakeQuatFromAxisAngle(Axis=<0 0 0>, AngleInDegrees=0)
     )
    )) {enabled=false, isSetParameters=true, module=SetVariables_#1}
   &Color(
    Color=<rgba 1 0.0833 0 1>
    Scale Color=&VectorFromCurve(VectorCurve=DataInterface() {propertyValues {
       XCurve {keys [<at -0.0031 48.8>,<at 1 1>]}
       YCurve {keys [<at -0.006 20.3>,<at 1 0.8>]}
       ZCurve {keys [<at -0.0011 1.0067>,<at 0.9989 0.28>]}
      }})
   )
   &ScaleMeshSize(Scale Factor=&VectorFromCurve(VectorCurve=DataInterface() {propertyValues {
       XCurve {keys [{arriveTangent=6.3957, interpMode=RCIM_Cubic, leaveTangent=6.3957, tangentMode=RCTM_User, time=0, value=0.47}, <at 0.9748 1.29>]}
       YCurve {keys [{arriveTangent=6.1645, interpMode=RCIM_Cubic, leaveTangent=6.1645, tangentMode=RCTM_User, time=0.0035, value=0.4903}, <at 0.9786 1.29>]}
       ZCurve {keys [
         <at -0.0011 1.0067 RCIM_Cubic>
         {arriveTangent=-1.635, interpMode=RCIM_Cubic, leaveTangent=-1.635, tangentMode=RCTM_User, time=1.0016, value=0.28}
        ]}
      }}))
   &CameraOffset(Camera Offset Amount=-10)
  ]
  rendererClasses [&NiagaraMeshRendererProperties]
  renderers [{rendererClass=&NiagaraMeshRendererProperties, rendererIndex=0}]
 }
 Light {
  emitterUpdateScript [
   &EmitterState()
   &SpawnRate(SpawnRate=4, Use Spawn Probability=false)
  ]
  particleSpawnScript [
   /Niagara/Modules/Spawn/Initialization/V2/InitializeParticle(Color Mode=::NewEnumerator0, Lifetime=0.55, UsePositionOffset=false, Write Material Random=true)
   &FN_SetSpawnGroup(Number of Spawn Groups=2)
   &Color(Color=<rgba 1.5 0.3116 0 1>)
  ]
  particleUpdateScript [
   &ParticleState()
   &SkeletalMeshLocation(
    Bone / Socket Index=$Particles.SpawnGroup
    Bone Sampling Mode=::NewEnumerator4
    MaskSpawnGroup=false
    Sampled Position Offset=<0 0 5>
    Sampled Position Offset Coordinate Space:ENiagaraSkelMeshTransforms=::NewEnumerator0
    Skeletal Mesh=$User.SkeletalMesh
   )
   &FN_Light_Attributes(
    Color=<rgba 80 24.5104 3.5329 1>
    Create Color Attribute=true
    Create Exponent Attribute=true
    Create Radius Attribute=true
    Exponent=4
    Radius=50
    Scale Color=&Lerp_Float(A=0.5, B=1, Alpha=$System.Dive)
   )
   &CameraOffset(Camera Offset Amount=3)
   &ScaleColor(
    Scale Alpha=&FloatFromCurve(FloatCurve=DataInterface() {propertyValues {Curve {keys [
         {arriveTangent=4.0162, interpMode=RCIM_Cubic, leaveTangent=4.0162, tangentMode=RCTM_User, time=0, value=0}
         {arriveTangent=0.2181, interpMode=RCIM_Cubic, leaveTangent=0.2181, tangentMode=RCTM_User, time=0.4, value=1}
         {arriveTangent=-0.8431, interpMode=RCIM_Cubic, leaveTangent=-0.8431, tangentMode=RCTM_User, time=1, value=0}
        ]}}})
    ScaleA=true
    ScaleRGB=true
   )
  ]
  rendererClasses [&NiagaraSpriteRendererProperties, &NiagaraLightRendererProperties]
  renderers [
   {rendererClass=&NiagaraSpriteRendererProperties}
   {rendererClass=&NiagaraLightRendererProperties, rendererIndex=1}
  ]
 }
 Mesh01 {
  emitterUpdateScript [
   &EmitterState(
    Inactive Response=::NewEnumerator1
    Life Cycle Mode=::NewEnumerator1
    Loop Behavior=::NewEnumerator1
    Loop Duration Mode=::NewEnumerator1
    UseLoopDelay=false
   )
   &SpawnBurst_Instantaneous(Spawn Count=1, Use Spawn Probability=false)
  ]
  particleSpawnScript [
   &InitializeParticle(Color Mode=::NewEnumerator0, UsePositionOffset=false, Write Material Random=true)
   &InitialMeshOrientation(
    Orientation Axis=<0 0 0>
    Orientation Vector=<0 0.5 0>
    Rotation=<0 0.5 0>
    Rotation Coordinate Space=::NewEnumerator2
    Use Orientation Vector=true
    Use Rotation Vector=true
   )
  ]
  particleUpdateScript [
   &ParticleState(Kill Particles When Lifetime Has Elapsed=false, Let Infinitely Lived Particles Die When Emitter Deactivates=true)
   &SkeletalMeshLocation(
    Bone Sampling Mode=::NewEnumerator0
    MaskSpawnGroup=false
    OverrideRandomSeed=false
    Sampled Position Offset=<10 1 0>
    Skeletal Mesh=DataInterface() {propertyValues {
      DefaultMesh=null
      ExcludeBoneName=root
      OptionalComponentClass=null
      PreviewMesh=&M_Med_Soldier_04_Preview
      SkinningMode=SkinOnTheFly
      SoftSourceActor=null
      SourceMode=Default
      UvSetIndex=0
      WholeMeshLOD=-1
      bExcludeBone=false
      bReadDeformedGeometry=true
      bRequireCurrentFrameData=true
      ComponentTags []
      FilteredBones [lowerarm_l]
      FilteredSockets []
      MeshUserParameter {parameter {name=None, typeDefHandle {registeredTypeIndex=1}, varData []}}
      SamplingRegions []
     }}
   )
   /SetVariables(Particles.MeshOrientation=&LerpQuaternion(
     Quaternion A:Quat4f=$Particles.MeshOrientation
     Quaternion B:Quat4f=&MultiplyQuaternion(
      Quaternion A=$Particles.MeshOrientation
      Quaternion B=&FindQuatBetweenTwoVectors(From Vector=<0 0 0>, To Vector=<0 0 1>)
     )
     Lerp Factor:NiagaraFloat=&Ceil_Float(Float=$System.Dive)
     Enable Spherical Linear Interpolation:NiagaraBool=false
    )) {enabled=false, isSetParameters=true}
   /SetVariables(Particles.NewOrientation:Quat4f=&MultiplyQuaternion(
     Quaternion A=$Particles.MeshOrientation
     Quaternion B=&LerpQuaternion(
      Quaternion A:Quat4f=<0 0 0 1>
      Quaternion B:Quat4f=<0 -0.5 0 1>
      Lerp Factor:NiagaraFloat=$System.Dive
      Enable Spherical Linear Interpolation:NiagaraBool=false
     )
    )) {isSetParameters=true}
  ]
  rendererClasses [&NiagaraMeshRendererProperties]
  renderers [{rendererClass=&NiagaraMeshRendererProperties, rendererIndex=0}]
 }
 Mesh02 {
  emitterUpdateScript [
   &EmitterState(
    Inactive Response=::NewEnumerator1
    Life Cycle Mode=::NewEnumerator1
    Loop Behavior=::NewEnumerator1
    Loop Duration Mode=::NewEnumerator1
    UseLoopDelay=false
   )
   &SpawnBurst_Instantaneous(Spawn Count=1, Use Spawn Probability=false)
  ]
  particleSpawnScript [
   &InitializeParticle(Color Mode=::NewEnumerator0, UsePositionOffset=false, Write Material Random=true)
   &InitialMeshOrientation(
    Orientation Axis=<0 0 0>
    Orientation Vector=<0 0.25 0>
    Rotation Coordinate Space=::NewEnumerator2
    Use Orientation Vector=true
    Use Rotation Vector=true
   )
  ]
  particleUpdateScript [
   &ParticleState(Kill Particles When Lifetime Has Elapsed=false, Let Infinitely Lived Particles Die When Emitter Deactivates=true)
   &UpdateMeshOrientation(Rotation Coordinate Space=::NewEnumerator2, Rotation Vector=<0.5 0.5 0.1>) {enabled=false}
   &SkeletalMeshLocation(
    Bone Sampling Mode=::NewEnumerator0
    MaskSpawnGroup=false
    OverrideRandomSeed=false
    Sampled Position Offset=<-10 0 -1>
    Skeletal Mesh=DataInterface() {propertyValues {
      DefaultMesh=null
      ExcludeBoneName=root
      OptionalComponentClass=null
      PreviewMesh=&M_Med_Soldier_04_Preview
      SkinningMode=SkinOnTheFly
      SoftSourceActor=null
      SourceMode=Default
      UvSetIndex=0
      WholeMeshLOD=-1
      bExcludeBone=false
      bReadDeformedGeometry=true
      bRequireCurrentFrameData=true
      ComponentTags []
      FilteredBones [lowerarm_r]
      FilteredSockets []
      MeshUserParameter {parameter {name=None, typeDefHandle {registeredTypeIndex=1}, varData []}}
      SamplingRegions []
     }}
   )
   /SetVariables(Particles.NewOrientation:Quat4f=&MultiplyQuaternion(
     Quaternion A=$Particles.MeshOrientation
     Quaternion B=&LerpQuaternion(
      Quaternion A:Quat4f=<0 0 0 1>
      Quaternion B:Quat4f=<0 0.5 0 1>
      Lerp Factor:NiagaraFloat=$System.Dive
      Enable Spherical Linear Interpolation:NiagaraBool=false
     )
    )) {isSetParameters=true}
  ]
  rendererClasses [&NiagaraMeshRendererProperties]
  renderers [{rendererClass=&NiagaraMeshRendererProperties, rendererIndex=0}]
 }
 Smoke {
  emitterUpdateScript [
   &EmitterState()
   &SpawnRate(SpawnRate=&Lerp_Float(A=30, B=0, Alpha=$System.Dive), Use Spawn Probability=false)
  ]
  particleSpawnScript [
   &FN_SetSpawnGroup(Number of Spawn Groups=2)
   &InitializeParticle(
    Color Mode=::NewEnumerator0
    Lifetime=&RandomRangeFloat(Minimum=0.4, Maximum=0.6)
    Sprite Rotation Angle=&RandomRangeFloat(Minimum=0, Maximum=1)
    Sprite Rotation Mode=::NewEnumerator1
    Sprite Size Mode=::NewEnumerator1
    Uniform Sprite Size Max=60
    Uniform Sprite Size Min=30
    UsePositionOffset=false
    Write Material Random=true
   )
   &SkeletalMeshLocation(
    Bone / Socket Index=$Particles.SpawnGroup
    Bone Sampling Mode=::NewEnumerator4
    Inherit Velocity=false
    MaskSpawnGroup=false
    Skeletal Mesh=$User.SkeletalMesh
   )
   &Color(Color=<rgba 0.2812 0.2812 0.2812 0.3>)
   &AddVelocityInCone(
    Cone Angle=8
    Cone Axis=<0 0 1>
    Cone Axis Coordinate Space=::Simulation
    Module.Use Velocity Falloff On Cone Axis=false
    Velocity Strength=&RandomRangeFloat(Minimum=200, Maximum=400)
   )
   &InheritVelocity(Module.Limit Speed=true, Velocity Scale=<1 0 0>) {enabled=false}
  ]
  particleUpdateScript [
   &Drag(Drag=4, Use Linear Drag=true, Use Rotational Drag=false)
   /Game/Effects/Niagara/Modules/Skydiving/FN_DivingAttributes/FN_DivingAttributes_v2(
    Accumulate Dive=1
    Accumulate Rest=-0.1
    Bone Index=2
    Sensitivity=1
    SkeletalMesh=DataInterface() {propertyValues {
      DefaultMesh=null
      ExcludeBoneName=root
      OptionalComponentClass=null
      PreviewMesh=&M_Med_Soldier_04_Preview
      SkinningMode=SkinOnTheFly
      SoftSourceActor=null
      SourceMode=Default
      UvSetIndex=0
      WholeMeshLOD=-1
      bExcludeBone=false
      bReadDeformedGeometry=true
      bRequireCurrentFrameData=true
      ComponentTags []
      FilteredBones [spine_02]
      FilteredSockets []
      MeshUserParameter {parameter {name=None, typeDefHandle {registeredTypeIndex=1}, varData []}}
      SamplingRegions []
     }}
   )
   &SolveForcesAndVelocity(Clamp Velocity=false, Limit Acceleration=false)
   &ParticleState()
   &DynamicMaterialParameters(
    Index 0 Param 1=&FloatFromCurve(FloatCurve=DataInterface() {propertyValues {Curve {keys [<at 0 1>,<at 1 2>]}}}, Scale Curve=0.2)
    Index 0 Param 3=&Lerp_Float(
     A=&FloatFromCurve(
      FloatCurve=DataInterface() {propertyValues {Curve {keys [
          {arriveTangent=-0.9531, interpMode=RCIM_Cubic, leaveTangent=-0.9531, tangentMode=RCTM_User, time=0, value=0.07}
          {arriveTangent=-0.3362, interpMode=RCIM_Cubic, leaveTangent=-0.3362, tangentMode=RCTM_User, time=0.1, value=0.01}
          {interpMode=RCIM_Cubic, leaveTangent=1.7757, time=0.15, value=0}
         ]}}}
      Scale Curve=0.35
     )
     B=&FloatFromCurve(
      FloatCurve=DataInterface() {propertyValues {Curve {keys [
          {arriveTangent=-1.0259, interpMode=RCIM_Cubic, leaveTangent=-1.0259, tangentMode=RCTM_User, time=0, value=0.07}
          {arriveTangent=-0.3308, interpMode=RCIM_Cubic, leaveTangent=-0.3308, tangentMode=RCTM_User, time=0.1, value=0.01}
          <at 0.16 0 RCIM_Cubic>
         ]}}}
      Scale Curve=0.4
     )
     Alpha=$System.Dive
    )
    Index 1 Param 1=&FloatFromCurve(FloatCurve=DataInterface() {propertyValues {Curve {keys [<at 0 0>,<at 1 1>]}}})
    Index 1 Param 2=&FloatFromCurve(FloatCurve=DataInterface() {propertyValues {Curve {keys [{arriveTangent=1.9694, interpMode=RCIM_Cubic, leaveTangent=1.9694, tangentMode=RCTM_User, time=0, value=0}, <at 1 1 RCIM_Cubic>]}}})
    Index 1 Param 3=&FloatFromCurve(FloatCurve=DataInterface() {propertyValues {Curve {keys [{arriveTangent=6.205, interpMode=RCIM_Cubic, leaveTangent=6.205, tangentMode=RCTM_User, time=0, value=0.2}, <at 0.25 1 RCIM_Cubic>]}}})
    Index 1 Param 4=&FloatFromCurve(FloatCurve=DataInterface() {propertyValues {Curve {keys [<at 0 1>,<at 1 0>]}}})
    Index 2 Param 1=&FloatFromCurve(FloatCurve=DataInterface() {propertyValues {Curve {keys [{arriveTangent=2.2804, interpMode=RCIM_Cubic, leaveTangent=2.2804, tangentMode=RCTM_User, time=0, value=0.1}, <at 1 1 RCIM_Cubic>]}}})
    Index 2 Param 2=&FloatFromCurve(FloatCurve=DataInterface() {propertyValues {Curve {keys [<at 0 0.2>,<at 1 0.8>]}}})
    Index 2 Param 3=&FloatFromCurve(FloatCurve=DataInterface() {propertyValues {Curve {keys [{interpMode=RCIM_Cubic, tangentMode=RCTM_User, time=0, value=1}, {interpMode=RCIM_Cubic, tangentMode=RCTM_User, time=1, value=2}]}}})
    Index 2 Param 4=&FloatFromCurve(FloatCurve=DataInterface() {propertyValues {Curve {keys [<at 0 1 RCIM_Cubic>,<at 1 0.75 RCIM_Cubic>]}}})
    Module.Param0WriteEnabled=true
    Module.Param0WriteEnabled1=true
    Module.Param0WriteEnabled2=true
    Module.Param0WriteEnabled3=true
    Module.Param1WriteEnabled=true
    Module.Param1WriteEnabled1=true
    Module.Param1WriteEnabled2=true
    Module.Param1WriteEnabled3=true
    Module.Param2WriteEnabled=true
    Module.Param2WriteEnabled1=true
    Module.Param2WriteEnabled2=true
    Module.Param2WriteEnabled3=true
    Module.Param3WriteEnabled=true
    Module.Param3WriteEnabled1=true
    Module.Param3WriteEnabled2=true
    Module.Param3WriteEnabled3=true
    Write Parameter Index 1=true
    Write Parameter Index 2=true
    Write Parameter Index 3=true
   )
   &ScaleSpriteSize(Scale Factor=&Vector2DFromFloat(Value=&FloatFromCurve(FloatCurve=DataInterface() {propertyValues {Curve {keys [
          {arriveTangent=1.1853, interpMode=RCIM_Cubic, leaveTangent=1.1853, tangentMode=RCTM_User, time=0, value=0.35}
          {arriveTangent=0.8143, interpMode=RCIM_Cubic, leaveTangent=0.8143, tangentMode=RCTM_User, time=0.3, value=0.625}
          {arriveTangent=-0.0405, interpMode=RCIM_Cubic, leaveTangent=-0.0405, tangentMode=RCTM_User, time=1, value=0.7}
         ]}}})))
   &ScaleColor(
    Scale Alpha=&FloatFromCurve(FloatCurve=DataInterface() {propertyValues {Curve {keys [
         {arriveTangent=3.7034, interpMode=RCIM_Cubic, leaveTangent=3.7034, tangentMode=RCTM_User, time=0, value=0}
         {arriveTangent=2.7145, interpMode=RCIM_Cubic, leaveTangent=2.7145, tangentMode=RCTM_User, time=0.3, value=0.95}
         <at 0.5 1>
         {arriveTangent=-0.1841, interpMode=RCIM_Cubic, leaveTangent=-0.1841, tangentMode=RCTM_User, time=1, value=0}
        ]}}})
    ScaleA=true
    ScaleRGB=true
   )
  ]
  rendererClasses [&NiagaraSpriteRendererProperties]
  renderers [{rendererClass=&NiagaraSpriteRendererProperties}]
 }
 Smoke1 {
  emitterUpdateScript [
   &EmitterState()
   &SpawnRate(SpawnRate=&Lerp_Float(A=0, B=90, Alpha=$System.Dive), Use Spawn Probability=false)
  ]
  particleSpawnScript [
   &FN_SetSpawnGroup(Number of Spawn Groups=2)
   &InitializeParticle(
    Color Mode=::NewEnumerator0
    Lifetime=&Lerp_Float(
     A=&RandomRangeFloat(Minimum=0.32, Maximum=0.5)
     B=&RandomRangeFloat(Minimum=0.5, Maximum=0.72)
     Alpha=$System.Dive
    )
    Sprite Rotation Angle=&RandomRangeFloat(Minimum=0, Maximum=1)
    Sprite Rotation Mode=::NewEnumerator1
    Sprite Size Mode=::NewEnumerator1
    Uniform Sprite Size Max=45
    Uniform Sprite Size Min=40
    UsePositionOffset=false
    Write Material Random=true
   )
   &Color(Color=<rgba 0.2552 0.2552 0.2552 0.4>)
   &AddVelocityInCone(
    Cone Angle=14
    Cone Axis=<0 0 1>
    Cone Axis Coordinate Space=::Simulation
    Module.Use Velocity Falloff On Cone Axis=false
    Velocity Strength=&RandomRangeFloat(Minimum=400, Maximum=720)
   ) {enabled=false}
   &InheritVelocity(Module.Limit Speed=true, Velocity Scale=<1 0 0>) {enabled=false}
  ]
  particleUpdateScript [
   &FN_DivingAttributes_v2(
    Accumulate Dive=0.1
    Accumulate Rest=-0.1
    Bone Index=2
    Sensitivity=1
    SkeletalMesh=DataInterface() {propertyValues {
      DefaultMesh=null
      ExcludeBoneName=root
      OptionalComponentClass=null
      PreviewMesh=&M_Med_Soldier_04_Preview
      SkinningMode=SkinOnTheFly
      SoftSourceActor=null
      SourceMode=Default
      UvSetIndex=0
      WholeMeshLOD=-1
      bExcludeBone=false
      bReadDeformedGeometry=true
      bRequireCurrentFrameData=true
      ComponentTags []
      FilteredBones [pelvis]
      FilteredSockets []
      MeshUserParameter {parameter {name=None, typeDefHandle {registeredTypeIndex=1}, varData []}}
      SamplingRegions []
     }}
   )
   &ParticleState()
   &SkeletalMeshLocation(
    Bone / Socket Index=$Particles.SpawnGroup
    Bone Sampling Mode=::NewEnumerator4
    Inherit Velocity=false
    Interpolate Mesh Sampling=false
    MaskSpawnGroup=false
    Sampled Position Offset=&Lerp_Vector(
     A=<0 0 0>
     B=&RandomRangeVector(Minimum=<10 10 145>, Maximum=<-10 -10 200>)
     Alpha=$Particles.NormalizedAge
    )
    Skeletal Mesh=$User.SkeletalMesh
   )
   &DynamicMaterialParameters(
    Index 0 Param 1=&FloatFromCurve(FloatCurve=DataInterface() {propertyValues {Curve {keys [<at 0 1>,<at 1 2>]}}})
    Index 0 Param 3=&FloatFromCurve(
     FloatCurve=DataInterface() {propertyValues {Curve {keys [
         {arriveTangent=-3.4683, interpMode=RCIM_Cubic, leaveTangent=-3.4683, tangentMode=RCTM_User, time=0, value=0.5}
         {arriveTangent=-2.0965, interpMode=RCIM_Cubic, leaveTangent=-2.0965, tangentMode=RCTM_User, time=0.16, value=0.18}
         <at 0.3 0>
        ]}}}
     Scale Curve=&Lerp_Float(A=0.12, B=0.7, Alpha=$System.Dive)
    )
    Index 1 Param 1=&FloatFromCurve(FloatCurve=DataInterface() {propertyValues {Curve {keys [<at 0 0>,<at 1 1>]}}})
    Index 1 Param 2=&FloatFromCurve(FloatCurve=DataInterface() {propertyValues {Curve {keys [{arriveTangent=1.9694, interpMode=RCIM_Cubic, leaveTangent=1.9694, tangentMode=RCTM_User, time=0, value=0}, <at 1 1 RCIM_Cubic>]}}})
    Index 1 Param 3=&FloatFromCurve(FloatCurve=DataInterface() {propertyValues {Curve {keys [{arriveTangent=6.205, interpMode=RCIM_Cubic, leaveTangent=6.205, tangentMode=RCTM_User, time=0, value=0.2}, <at 0.25 1 RCIM_Cubic>]}}})
    Index 1 Param 4=&FloatFromCurve(FloatCurve=DataInterface() {propertyValues {Curve {keys [<at 0 1>,<at 1 0>]}}})
    Index 2 Param 1=&FloatFromCurve(FloatCurve=DataInterface() {propertyValues {Curve {keys [{arriveTangent=2.2804, interpMode=RCIM_Cubic, leaveTangent=2.2804, tangentMode=RCTM_User, time=0, value=0.1}, <at 1 1 RCIM_Cubic>]}}})
    Index 2 Param 2=&FloatFromCurve(FloatCurve=DataInterface() {propertyValues {Curve {keys [<at 0 0.2>,<at 1 0.8>]}}})
    Index 2 Param 3=&FloatFromCurve(FloatCurve=DataInterface() {propertyValues {Curve {keys [{interpMode=RCIM_Cubic, tangentMode=RCTM_User, time=0, value=1}, {interpMode=RCIM_Cubic, tangentMode=RCTM_User, time=1, value=2}]}}})
    Index 2 Param 4=&FloatFromCurve(FloatCurve=DataInterface() {propertyValues {Curve {keys [<at 0 1 RCIM_Cubic>,<at 1 0.75 RCIM_Cubic>]}}})
    Module.Param0WriteEnabled=true
    Module.Param0WriteEnabled1=true
    Module.Param0WriteEnabled2=true
    Module.Param0WriteEnabled3=true
    Module.Param1WriteEnabled=true
    Module.Param1WriteEnabled1=true
    Module.Param1WriteEnabled2=true
    Module.Param1WriteEnabled3=true
    Module.Param2WriteEnabled=true
    Module.Param2WriteEnabled1=true
    Module.Param2WriteEnabled2=true
    Module.Param2WriteEnabled3=true
    Module.Param3WriteEnabled=true
    Module.Param3WriteEnabled1=true
    Module.Param3WriteEnabled2=true
    Module.Param3WriteEnabled3=true
    Write Parameter Index 1=true
    Write Parameter Index 2=true
    Write Parameter Index 3=true
   )
   &ScaleSpriteSize(Scale Factor=&Vector2DFromFloat(Value=&FloatFromCurve(FloatCurve=DataInterface() {propertyValues {Curve {keys [
          {arriveTangent=1.0159, interpMode=RCIM_Cubic, leaveTangent=1.0159, tangentMode=RCTM_User, time=0, value=0.35}
          {arriveTangent=0.3784, interpMode=RCIM_Cubic, leaveTangent=0.3784, tangentMode=RCTM_User, time=0.3, value=0.575}
          {arriveTangent=-0.0405, interpMode=RCIM_Cubic, leaveTangent=-0.0405, tangentMode=RCTM_User, time=1, value=0.75}
         ]}}})))
   &ScaleColor(
    Scale Alpha=&FloatFromCurve(FloatCurve=DataInterface() {propertyValues {Curve {keys [
         {arriveTangent=1.9541, interpMode=RCIM_Cubic, leaveTangent=1.9541, tangentMode=RCTM_User, time=0, value=0}
         {arriveTangent=1.7161, interpMode=RCIM_Cubic, leaveTangent=1.7161, tangentMode=RCTM_User, time=0.45, value=1}
         {arriveTangent=-0.1841, interpMode=RCIM_Cubic, leaveTangent=-0.1841, tangentMode=RCTM_User, time=1, value=0}
        ]}}})
    ScaleA=true
    ScaleRGB=true
   )
   &Drag(Drag=4, Use Linear Drag=true, Use Rotational Drag=false) {enabled=false}
   &SolveForcesAndVelocity(Clamp Velocity=false, Limit Acceleration=false)
  ]
  rendererClasses [&NiagaraSpriteRendererProperties]
  renderers [{rendererClass=&NiagaraSpriteRendererProperties}]
 }
 Smoke2 {
  emitterUpdateScript [
   &EmitterState()
   &SpawnRate(SpawnRate=&Lerp_Float(A=0, B=120, Alpha=$System.Dive), Use Spawn Probability=false)
  ]
  particleSpawnScript [
   &FN_SetSpawnGroup(Number of Spawn Groups=2)
   &InitializeParticle(
    Color Mode=::NewEnumerator0
    Lifetime=&RandomRangeFloat(Minimum=0.55, Maximum=0.85)
    Sprite Rotation Angle=&RandomRangeFloat(Minimum=0, Maximum=1)
    Sprite Rotation Mode=::NewEnumerator1
    Sprite Size Mode=::NewEnumerator1
    Uniform Sprite Size Max=45
    Uniform Sprite Size Min=40
    UsePositionOffset=false
    Write Material Random=true
   )
   &SkeletalMeshLocation(
    Bone / Socket Index=$Particles.SpawnGroup
    Bone Sampling Mode=::NewEnumerator4
    Inherit Velocity=false
    Interpolate Mesh Sampling=false
    MaskSpawnGroup=false
    Skeletal Mesh=$User.SkeletalMesh
   )
   &Color(Color=<rgba 0.2812 0.2812 0.2812 0.4>)
   &AddVelocityInCone(
    Cone Angle=4
    Cone Axis=<0 0 1>
    Module.Use Velocity Falloff On Cone Axis=false
    Velocity Strength=&RandomRangeFloat(Minimum=550, Maximum=860)
   )
   &InheritVelocity(Module.Limit Speed=true) {module=InheritVelocity001}
  ]
  particleUpdateScript [
   &Drag(Drag=3.5, Use Linear Drag=true, Use Rotational Drag=false)
   &FN_DivingAttributes_v2(
    Accumulate Dive=1
    Accumulate Rest=-0.1
    Bone Index=2
    Sensitivity=1
    SkeletalMesh=DataInterface() {propertyValues {
      DefaultMesh=null
      ExcludeBoneName=root
      OptionalComponentClass=null
      PreviewMesh=&M_Med_Soldier_04_Preview
      SkinningMode=SkinOnTheFly
      SoftSourceActor=null
      SourceMode=Default
      UvSetIndex=0
      WholeMeshLOD=-1
      bExcludeBone=false
      bReadDeformedGeometry=true
      bRequireCurrentFrameData=true
      ComponentTags []
      FilteredBones [spine_02]
      FilteredSockets []
      MeshUserParameter {parameter {name=None, typeDefHandle {registeredTypeIndex=1}, varData []}}
      SamplingRegions []
     }}
   )
   &SolveForcesAndVelocity(Clamp Velocity=false, Limit Acceleration=false)
   &ParticleState()
   &DynamicMaterialParameters(
    Index 0 Param 1=&FloatFromCurve(FloatCurve=DataInterface() {propertyValues {Curve {keys [<at 0 1>,<at 1 2>]}}}, Scale Curve=0.2)
    Index 0 Param 3=&Lerp_Float(
     A=&FloatFromCurve(
      FloatCurve=DataInterface() {propertyValues {Curve {keys [
          {arriveTangent=-0.9531, interpMode=RCIM_Cubic, leaveTangent=-0.9531, tangentMode=RCTM_User, time=0, value=0.07}
          {arriveTangent=-0.2556, interpMode=RCIM_Cubic, leaveTangent=-0.2556, tangentMode=RCTM_User, time=0.1, value=0.01}
          {interpMode=RCIM_Cubic, leaveTangent=1.7757, time=0.15, value=0}
         ]}}}
      Scale Curve=0.35
     )
     B=&FloatFromCurve(
      FloatCurve=DataInterface() {propertyValues {Curve {keys [
          {arriveTangent=-5.5157, interpMode=RCIM_Cubic, leaveTangent=-5.5157, tangentMode=RCTM_User, time=0, value=1}
          {arriveTangent=-3.0904, interpMode=RCIM_Cubic, leaveTangent=-3.0904, tangentMode=RCTM_User, time=0.16, value=0.25}
          <at 0.3 0>
         ]}}}
      Scale Curve=&Lerp_Float(A=0.12, B=0.7, Alpha=$System.Dive)
     )
     Alpha=$System.Dive
    )
    Index 1 Param 1=&FloatFromCurve(FloatCurve=DataInterface() {propertyValues {Curve {keys [<at 0 0>,<at 1 1>]}}})
    Index 1 Param 2=&FloatFromCurve(FloatCurve=DataInterface() {propertyValues {Curve {keys [{arriveTangent=1.9694, interpMode=RCIM_Cubic, leaveTangent=1.9694, tangentMode=RCTM_User, time=0, value=0}, <at 1 1 RCIM_Cubic>]}}})
    Index 1 Param 3=&FloatFromCurve(FloatCurve=DataInterface() {propertyValues {Curve {keys [{arriveTangent=6.205, interpMode=RCIM_Cubic, leaveTangent=6.205, tangentMode=RCTM_User, time=0, value=0.2}, <at 0.25 1 RCIM_Cubic>]}}})
    Index 1 Param 4=&FloatFromCurve(FloatCurve=DataInterface() {propertyValues {Curve {keys [<at 0 1>,<at 1 0>]}}})
    Index 2 Param 1=&FloatFromCurve(FloatCurve=DataInterface() {propertyValues {Curve {keys [{arriveTangent=2.2804, interpMode=RCIM_Cubic, leaveTangent=2.2804, tangentMode=RCTM_User, time=0, value=0.1}, <at 1 1 RCIM_Cubic>]}}})
    Index 2 Param 2=&FloatFromCurve(FloatCurve=DataInterface() {propertyValues {Curve {keys [<at 0 0.2>,<at 1 0.8>]}}})
    Index 2 Param 3=&FloatFromCurve(FloatCurve=DataInterface() {propertyValues {Curve {keys [{interpMode=RCIM_Cubic, tangentMode=RCTM_User, time=0, value=1}, {interpMode=RCIM_Cubic, tangentMode=RCTM_User, time=1, value=2}]}}})
    Index 2 Param 4=&FloatFromCurve(FloatCurve=DataInterface() {propertyValues {Curve {keys [<at 0 1 RCIM_Cubic>,<at 1 0.75 RCIM_Cubic>]}}})
    Module.Param0WriteEnabled=true
    Module.Param0WriteEnabled1=true
    Module.Param0WriteEnabled2=true
    Module.Param0WriteEnabled3=true
    Module.Param1WriteEnabled=true
    Module.Param1WriteEnabled1=true
    Module.Param1WriteEnabled2=true
    Module.Param1WriteEnabled3=true
    Module.Param2WriteEnabled=true
    Module.Param2WriteEnabled1=true
    Module.Param2WriteEnabled2=true
    Module.Param2WriteEnabled3=true
    Module.Param3WriteEnabled=true
    Module.Param3WriteEnabled1=true
    Module.Param3WriteEnabled2=true
    Module.Param3WriteEnabled3=true
    Write Parameter Index 1=true
    Write Parameter Index 2=true
    Write Parameter Index 3=true
   )
   &ScaleSpriteSize(Scale Factor=&Vector2DFromFloat(Value=&FloatFromCurve(FloatCurve=DataInterface() {propertyValues {Curve {keys [
          {arriveTangent=1.1911, interpMode=RCIM_Cubic, leaveTangent=1.1911, tangentMode=RCTM_User, time=0, value=0.35}
          {arriveTangent=0.33, interpMode=RCIM_Cubic, leaveTangent=0.33, tangentMode=RCTM_User, time=0.4, value=0.6}
          {arriveTangent=-0.0405, interpMode=RCIM_Cubic, leaveTangent=-0.0405, tangentMode=RCTM_User, time=1, value=0.75}
         ]}}})))
   &ScaleColor(
    Scale Alpha=&FloatFromCurve(FloatCurve=DataInterface() {propertyValues {Curve {keys [
         {arriveTangent=1.9541, interpMode=RCIM_Cubic, leaveTangent=1.9541, tangentMode=RCTM_User, time=0, value=0}
         {arriveTangent=2.7145, interpMode=RCIM_Cubic, leaveTangent=2.7145, tangentMode=RCTM_User, time=0.3, value=1}
         <at 0.5 1>
         {arriveTangent=-0.1841, interpMode=RCIM_Cubic, leaveTangent=-0.1841, tangentMode=RCTM_User, time=1, value=0}
        ]}}})
    ScaleA=true
    ScaleRGB=true
   )
  ]
  rendererClasses [&NiagaraSpriteRendererProperties]
  renderers [{rendererClass=&NiagaraSpriteRendererProperties}]
 }
 Spark {
  emitterUpdateScript [
   &EmitterState()
   &SpawnRate(SpawnRate=&Lerp_Float(A=0, B=32, Alpha=$System.Dive), Use Spawn Probability=false) {module=SpawnRate001}
  ]
  particleSpawnScript [
   &InitializeParticle(
    Color Mode=::NewEnumerator0
    Lifetime Max=0.12
    Lifetime Min=0.08
    Lifetime Mode=::NewEnumerator1
    Sprite Size=<1 3.6>
    Sprite Size Mode=::NewEnumerator4
    UsePositionOffset=false
    Write Material Random=true
   )
   &FN_SetSpawnGroup(Number of Spawn Groups=2)
   &Color(Color=<rgba 500 111.6043 0 1>)
   &SkeletalMeshLocation(
    Bone / Socket Index=$Particles.SpawnGroup
    Bone Sampling Mode=::NewEnumerator4
    Interpolate Mesh Sampling=false
    MaskSpawnGroup=false
    Sampled Position Offset Coordinate Space:ENiagaraSkelMeshTransforms=::NewEnumerator2
    Skeletal Mesh=$User.SkeletalMesh
   )
   &AddVelocityInCone(
    Cone Angle=15
    Cone Axis=<0 0 1>
    Cone Axis Coordinate Space=::Simulation
    Module.Use Velocity Falloff On Cone Axis=false
    Velocity Strength=&RandomRangeFloat(Minimum=300, Maximum=800)
   )
   &InheritVelocity(Module.Limit Speed=true, Velocity Scale=<0.3 0.3 0.3>)
  ]
  particleUpdateScript [
   &Drag(Drag=10, Use Linear Drag=true, Use Rotational Drag=false)
   &VectorNoiseForce(Force Amount=10, Module.Use Random Noise Time Interval=false)
   &SolveForcesAndVelocity(Clamp Velocity=false, Limit Acceleration=false)
   &ParticleState()
   &ScaleSpriteSizeBySpeed(Max Scale Factor=<1.4 4>)
   &Color(Color=&ColorFromCurve(ColorCurve=DataInterface() {propertyValues {
       AlphaCurve {keys [<at 0 1 RCIM_Cubic>,<at 1 0 RCIM_Cubic>]}
       BlueCurve {keys [<at 0 4.6804 RCIM_Cubic>,<at 1 0 RCIM_Cubic>]}
       GreenCurve {keys [<at 0 0 RCIM_Cubic>,<at 1 28.1447 RCIM_Cubic>]}
       RedCurve {keys [<at 0 500 RCIM_Cubic>,<at 1 100 RCIM_Cubic>]}
      }})) {module=Color001}
  ]
  rendererClasses [&NiagaraSpriteRendererProperties]
  renderers [{rendererClass=&NiagaraSpriteRendererProperties}]
 }
 Spark1 {
  emitterUpdateScript [
   &EmitterState()
   &SpawnRate(SpawnRate=&Lerp_Float(A=0, B=40, Alpha=$System.Dive), Use Spawn Probability=false) {module=SpawnRate001}
  ]
  particleSpawnScript [
   &InitializeParticle(
    Color Mode=::NewEnumerator0
    Lifetime Max=0.6
    Lifetime Min=0.3
    Lifetime Mode=::NewEnumerator1
    Sprite Size=<0.8 2.8>
    Sprite Size Mode=::NewEnumerator4
    UsePositionOffset=false
    Write Material Random=true
   )
   &FN_SetSpawnGroup(Number of Spawn Groups=2)
   &Color(Color=<rgba 500 111.6043 0 1>)
   &SkeletalMeshLocation(
    Bone / Socket Index=$Particles.SpawnGroup
    Bone Sampling Mode=::NewEnumerator4
    Interpolate Mesh Sampling=false
    MaskSpawnGroup=false
    Sampled Position Offset Coordinate Space:ENiagaraSkelMeshTransforms=::NewEnumerator2
    Skeletal Mesh=$User.SkeletalMesh
   )
   &AddVelocityInCone(
    Cone Angle=12
    Cone Axis=<0 0 1>
    Cone Axis Coordinate Space=::Simulation
    Module.Use Velocity Falloff On Cone Axis=false
    Velocity Strength=&RandomRangeFloat(Minimum=400, Maximum=750)
   )
   &InheritVelocity(Module.Limit Speed=true, Velocity Scale=<0.3 0.3 0.3>)
  ]
  particleUpdateScript [
   &Drag(Drag=2, Use Linear Drag=true, Use Rotational Drag=false)
   &VectorNoiseForce(Module.Use Random Noise Time Interval=false)
   &GravityForce(Gravity=<0 0 -300>)
   &SolveForcesAndVelocity(Clamp Velocity=false, Limit Acceleration=false)
   &ParticleState()
   &ScaleSpriteSizeBySpeed(Max Scale Factor=<1.4 4>)
   &Color(Color=&ColorFromCurve(ColorCurve=DataInterface() {propertyValues {
       AlphaCurve {keys [<at 0 1 RCIM_Cubic>,<at 1 0 RCIM_Cubic>]}
       BlueCurve {keys [<at 0 4.6802 RCIM_Cubic>,<at 1 0 RCIM_Cubic>]}
       GreenCurve {keys [<at 0 0 RCIM_Cubic>,<at 1 28.1447 RCIM_Cubic>]}
       RedCurve {keys [<at 0 500 RCIM_Cubic>,<at 1 100 RCIM_Cubic>]}
      }})) {module=Color001}
  ]
  rendererClasses [&NiagaraSpriteRendererProperties]
  renderers [{rendererClass=&NiagaraSpriteRendererProperties}]
 }
 Syrup_Ribbons {
  emitterUpdateScript [
   &EmitterState()
   &FN_SpawnRibbon_v2(
    Number of Ribbons=2
    Spawn Mode=::NewEnumerator0
    SpawnRate:NiagaraFloat=&Lerp_Float(A=0, B=40, Alpha=$System.Dive)
   ) {module=FN_SpawnRibbon_v2001}
   &FN_SpawnRibbon_v2(Number of Ribbons=2, Spawn Mode=::NewEnumerator0) {enabled=false}
   &SpawnRate(SpawnRate=0, Use Spawn Probability=false)
  ]
  particleSpawnScript [
   &FN_InitializeRibbon_v3(
    Enable Skeletal Position=true
    Lifetime=0.7
    Ribbon Width=30
    Sampled Position Coordinate Space:ENiagaraCoordinateSpace=::Local
    Sampled Rotation Coordinate Space:ENiagaraCoordinateSpace=::Local
    Sampled Velocity Coordinate Space:ENiagaraCoordinateSpace=::Local
   ) {module=FN_InitializeRibbon_v3001}
   &AddVelocity(
    Coordinate Space=::Simulation
    Velocity=&MakeCustomVectorFromBool(
     Bool=$Particles.Locked
     True Vector=<0 0 0>
     False Vector=&FN_FrontEndVector3Modifier(Gameplay=<0 0 370>, Front End=<0 0 350>)
    )
   ) {module=AddVelocity001}
   &DynamicMaterialParameters(
    Index 0 Param 1=&MaskFloatBySpawnGroup(Masked Float=0, Passthrough Float=0.5)
    Module.Param0WriteEnabled=true
    Module.Param1WriteEnabled=false
    Module.Param2WriteEnabled=false
    Module.Param3WriteEnabled=false
   ) {enabled=false, module=DynamicMaterialParameters001}
  ]
  particleUpdateScript [
   &Drag(Drag=0.9, Use Linear Drag=true, Use Rotational Drag=false)
   &InheritVelocity(
    Module.Limit Speed=true
    Velocity Limit=10000
    Velocity Scale=&MakeCustomVectorFromBool(Bool=$Particles.Initial.Locked, True Vector=<0 0 0>, False Vector=<-0.03 -0.03 -0.01>)
    Velocity Source=&TransformVector(Vector=$Engine.Owner.Velocity, Source Space=::World, Destination Space=::Local)
   ) {enabled=false}
   &AddVelocity(Velocity=&Multiply_VectorByFloat(
     Vector=<0 0 1>
     Float=&FloatFromCurve(FloatCurve=DataInterface() {propertyValues {Curve {keys [<at 0 0 RCIM_Cubic>, {arriveTangent=2.8829, interpMode=RCIM_Cubic, leaveTangent=2.8829, tangentMode=RCTM_User, time=1, value=1}]}}})
    )) {module=AddVelocity002}
   &SolveForcesAndVelocity(Clamp Velocity=false, Limit Acceleration=false)
   &FN_UpdateRibbon(
    Enable Skeletal Position=true
    Initial Color:LinearColor=$Particles.Initial.Color
    Initial RibbonWidth:NiagaraFloat=$Particles.Initial.RibbonWidth
    Sampled Position Coordinate Space:ENiagaraCoordinateSpace=::Local
    Sampled Rotation Coordinate Space:ENiagaraCoordinateSpace=::Local
    Sampled Velocity Coordinate Space:ENiagaraCoordinateSpace=::Local
   )
   &Material(Base Texture Time:NiagaraFloat=&Multiply_Float(A=$Emitter.Age, B=-1))
   &DynamicMaterialParameters(
    Index 0 Param 1=1
    Index 0 Param 2=1
    Index 0 Param 3=0.1
    Index 0 Param 4=0.2
    Index 1 Param 1=&Multiply_Float(A=-1.6, B=$Emitter.Age)
    Index 1 Param 2=0.15
    Index 1 Param 3=1.3
    Index 1 Param 4=1.8
    Index 2 Param 1=0.2
    Index 2 Param 2=0.2
    Index 2 Param 3=0.2
    Index 3 Param 1=1
    Index 3 Param 2=1
    Index 3 Param 3=1
    Index 3 Param 4=1
    Module.Param0WriteEnabled=true
    Module.Param0WriteEnabled1=true
    Module.Param0WriteEnabled2=true
    Module.Param0WriteEnabled3=true
    Module.Param1WriteEnabled=true
    Module.Param1WriteEnabled1=true
    Module.Param1WriteEnabled2=true
    Module.Param1WriteEnabled3=true
    Module.Param2WriteEnabled=true
    Module.Param2WriteEnabled1=true
    Module.Param2WriteEnabled2=true
    Module.Param2WriteEnabled3=true
    Module.Param3WriteEnabled=true
    Module.Param3WriteEnabled1=true
    Module.Param3WriteEnabled2=true
    Module.Param3WriteEnabled3=true
    Write Parameter Index 1=true
    Write Parameter Index 2=true
    Write Parameter Index 3=true
   )
   &ScaleRibbonWidth(Ribbon Width Scale=&FloatFromCurve(
     FloatCurve=DataInterface() {propertyValues {Curve {keys [
         {arriveTangent=-0.0063, interpMode=RCIM_Cubic, leaveTangent=-0.0063, tangentMode=RCTM_User, time=0.0058, value=1}
         {arriveTangent=1.4891, interpMode=RCIM_Cubic, leaveTangent=1.4891, time=0.4, value=1.4}
         <at 0.6774 2>
        ]}}}
     Scale Curve=0.6
    ))
   &Color(
    Color=&ColorFromCurve(ColorCurve=DataInterface() {propertyValues {
       AlphaCurve {keys [<at 0 0 RCIM_Cubic>,<at 0.0787 2>,<at 0.177 15>,<at 0.6713 1>,<at 0.9977 0 RCIM_Cubic>]}
       BlueCurve {keys [<at 0.0098 0.1458>,<at 0.0615 0>,<at 0.1704 0.1094>,<at 0.3393 0.0521>,<at 0.7967 0.001>,<at 1 0 RCIM_Cubic>]}
       GreenCurve {keys [<at 0.0098 0.2734>,<at 0.0615 0.2145>,<at 0.1704 0.1094>,<at 0.3393 0.0521>,<at 0.7967 0.0052>,<at 1 0 RCIM_Cubic>]}
       RedCurve {keys [<at 0.0098 2>,<at 0.0615 0.7656>,<at 0.1704 0.1094>,<at 0.3393 0.0521>,<at 0.7967 0.0156>,<at 1 0 RCIM_Cubic>]}
      }})
    Scale Alpha=&Multiply_Float(
     A=&FloatFromCurve(
      FloatCurve=DataInterface() {propertyValues {Curve {keys [
          {arriveTangent=-2.9592, interpMode=RCIM_Cubic, leaveTangent=-2.9592, tangentMode=RCTM_User, time=-0.0046, value=1.0113}
          {arriveTangent=-1.4276, interpMode=RCIM_Cubic, leaveTangent=-1.4276, tangentMode=RCTM_User, time=0.3, value=0.3}
          <at 1.0011 0.002>
         ]}}}
      Scale Curve=&FloatFromCurve(FloatCurve=DataInterface() {propertyValues {Curve {keys [<at 0.3 0>,<at 0.9974 0.1>]}}}, CurveIndex=$Emitter.Age)
     )
     B=&FNTimeOfDayScalar(Night Scale=1, Morning Scale=0.3, Evening Scale=0.3)
    )
    Scale Color=<4 4 1>
   )
  ]
  rendererClasses [&NiagaraRibbonRendererProperties]
  renderers [{rendererClass=&NiagaraRibbonRendererProperties}]
 }
 Syrup_Ribbons001 {
  emitterUpdateScript [
   &EmitterState()
   &FN_SpawnRibbon_v2(
    Number of Ribbons=2
    Spawn Mode=::NewEnumerator0
    SpawnRate:NiagaraFloat=&Lerp_Float(A=0, B=100, Alpha=$System.Dive)
   ) {module=FN_SpawnRibbon_v2001}
   &FN_SpawnRibbon_v2(Number of Ribbons=2, Spawn Mode=::NewEnumerator0) {enabled=false}
   &SpawnRate(SpawnRate=0, Use Spawn Probability=false)
  ]
  particleSpawnScript [
   &FN_InitializeRibbon_v3(
    Enable Skeletal Position=true
    Lifetime=1
    Ribbon Width=75
    Sampled Position Coordinate Space:ENiagaraCoordinateSpace=::Local
    Sampled Rotation Coordinate Space:ENiagaraCoordinateSpace=::Local
    Sampled Velocity Coordinate Space:ENiagaraCoordinateSpace=::Local
   ) {module=FN_InitializeRibbon_v3001}
   &AddVelocity(
    Coordinate Space=::Simulation
    Velocity=&MakeCustomVectorFromBool(
     Bool=$Particles.Locked
     True Vector=<0 0 0>
     False Vector=&FN_FrontEndVector3Modifier(Gameplay=<0 0 500>, Front End=<0 0 350>)
    )
   ) {module=AddVelocity001}
   &DynamicMaterialParameters(
    Index 0 Param 1=&MaskFloatBySpawnGroup(Masked Float=0, Passthrough Float=0.5)
    Module.Param0WriteEnabled=true
    Module.Param1WriteEnabled=false
    Module.Param2WriteEnabled=false
    Module.Param3WriteEnabled=false
   ) {enabled=false, module=DynamicMaterialParameters001}
  ]
  particleUpdateScript [
   &Drag(Drag=0.9, Use Linear Drag=true, Use Rotational Drag=false)
   &InheritVelocity(
    Module.Limit Speed=true
    Velocity Limit=10000
    Velocity Scale=&MakeCustomVectorFromBool(Bool=$Particles.Initial.Locked, True Vector=<0 0 0>, False Vector=<-0.03 -0.03 -0.01>)
    Velocity Source=&TransformVector(Vector=$Engine.Owner.Velocity, Source Space=::World, Destination Space=::Local)
   ) {enabled=false}
   &AddVelocity(Velocity=&Multiply_VectorByFloat(
     Vector=<0 0 1>
     Float=&FloatFromCurve(FloatCurve=DataInterface() {propertyValues {Curve {keys [<at 0 0 RCIM_Cubic>, {arriveTangent=2.8829, interpMode=RCIM_Cubic, leaveTangent=2.8829, tangentMode=RCTM_User, time=1, value=1}]}}})
    )) {module=AddVelocity002}
   &SolveForcesAndVelocity(Clamp Velocity=false, Limit Acceleration=false)
   &FN_UpdateRibbon(
    Enable Skeletal Position=true
    Initial Color:LinearColor=$Particles.Initial.Color
    Initial RibbonWidth:NiagaraFloat=$Particles.Initial.RibbonWidth
    Sampled Position Coordinate Space:ENiagaraCoordinateSpace=::Local
    Sampled Rotation Coordinate Space:ENiagaraCoordinateSpace=::Local
    Sampled Velocity Coordinate Space:ENiagaraCoordinateSpace=::Local
   )
   &Material(Base Texture Time:NiagaraFloat=&Multiply_Float(A=$Emitter.Age, B=-1))
   &DynamicMaterialParameters(
    Index 0 Param 1=1
    Index 0 Param 2=1
    Index 0 Param 3=0.1
    Index 0 Param 4=0.2
    Index 1 Param 1=&Multiply_Float(A=-1.6, B=$Emitter.Age)
    Index 1 Param 2=0.15
    Index 1 Param 3=1.3
    Index 1 Param 4=1.8
    Index 2 Param 1=0.2
    Index 2 Param 2=0.2
    Index 2 Param 3=0.2
    Index 3 Param 1=1
    Index 3 Param 2=1
    Index 3 Param 3=1
    Index 3 Param 4=1
    Module.Param0WriteEnabled=true
    Module.Param0WriteEnabled1=true
    Module.Param0WriteEnabled2=true
    Module.Param0WriteEnabled3=true
    Module.Param1WriteEnabled=true
    Module.Param1WriteEnabled1=true
    Module.Param1WriteEnabled2=true
    Module.Param1WriteEnabled3=true
    Module.Param2WriteEnabled=true
    Module.Param2WriteEnabled1=true
    Module.Param2WriteEnabled2=true
    Module.Param2WriteEnabled3=true
    Module.Param3WriteEnabled=true
    Module.Param3WriteEnabled1=true
    Module.Param3WriteEnabled2=true
    Module.Param3WriteEnabled3=true
    Write Parameter Index 1=true
    Write Parameter Index 2=true
    Write Parameter Index 3=true
   )
   &ScaleRibbonWidth(Ribbon Width Scale=&FloatFromCurve(
     FloatCurve=DataInterface() {propertyValues {Curve {keys [
         {arriveTangent=-0.0063, interpMode=RCIM_Cubic, leaveTangent=-0.0063, tangentMode=RCTM_User, time=0.0058, value=1}
         {arriveTangent=1.4891, interpMode=RCIM_Cubic, leaveTangent=1.4891, time=0.4, value=1.4}
         <at 0.6774 2>
        ]}}}
     Scale Curve=0.6
    ))
   &Color(
    Color=&ColorFromCurve(ColorCurve=DataInterface() {propertyValues {
       AlphaCurve {keys [<at 0 0 RCIM_Cubic>,<at 0.0787 2>,<at 0.177 15>,<at 0.6713 1>,<at 0.9977 0 RCIM_Cubic>]}
       BlueCurve {keys [<at 0.0098 0.1458>,<at 0.0615 0>,<at 0.1475 0>,<at 0.3393 0>,<at 0.7967 0.001>,<at 1 0 RCIM_Cubic>]}
       GreenCurve {keys [<at 0.0098 0.2734>,<at 0.0615 0.2145>,<at 0.1475 0.0092>,<at 0.3393 0>,<at 0.7967 0.0052>,<at 1 0 RCIM_Cubic>]}
       RedCurve {keys [<at 0.0098 2>,<at 0.0615 0.7656>,<at 0.1475 0.0625>,<at 0.3393 0>,<at 0.7967 0.0156>,<at 1 0 RCIM_Cubic>]}
      }})
    Scale Alpha=&Multiply_Float(
     A=&FloatFromCurve(
      FloatCurve=DataInterface() {propertyValues {Curve {keys [{arriveTangent=-2.0086, interpMode=RCIM_Cubic, leaveTangent=-2.0086, tangentMode=RCTM_User, time=-0.0046, value=1.0113}, <at 1.0011 0.002>]}}}
      Scale Curve=&FloatFromCurve(FloatCurve=DataInterface() {propertyValues {Curve {keys [<at 0.3 0>,<at 0.9974 0.1>]}}}, CurveIndex=$Emitter.Age)
     )
     B=&FNTimeOfDayScalar(Night Scale=1, Morning Scale=0.3, Evening Scale=0.3)
    )
    Scale Color=<4 4 1>
   )
  ]
  rendererClasses [&NiagaraRibbonRendererProperties]
  renderers [{rendererClass=&NiagaraRibbonRendererProperties}]
 }
 Syrup_Ribbons002 {
  emitterUpdateScript [
   &EmitterState()
   &FN_SpawnRibbon_v2(
    Number of Ribbons=2
    Spawn Mode=::NewEnumerator0
    SpawnRate:NiagaraFloat=&Lerp_Float(A=5, B=20, Alpha=$System.Dive)
   ) {module=FN_SpawnRibbon_v2001}
  ]
  particleSpawnScript [
   &FN_InitializeRibbon_v3(
    Enable Skeletal Position=true
    Lifetime=0.8
    Ribbon Width=50
    Sampled Position Coordinate Space:ENiagaraCoordinateSpace=::Local
    Sampled Rotation Coordinate Space:ENiagaraCoordinateSpace=::Local
    Sampled Velocity Coordinate Space:ENiagaraCoordinateSpace=::Local
   ) {module=FN_InitializeRibbon_v3001}
   &AddVelocity(
    Coordinate Space=::Simulation
    Velocity=&MakeCustomVectorFromBool(
     Bool=$Particles.Locked
     True Vector=<0 0 0>
     False Vector=&FN_FrontEndVector3Modifier(Gameplay=<0 0 400>, Front End=<0 0 350>)
    )
   ) {module=AddVelocity001}
   &DynamicMaterialParameters(
    Index 0 Param 1=&MaskFloatBySpawnGroup(Masked Float=0, Passthrough Float=0.5)
    Module.Param0WriteEnabled=true
    Module.Param1WriteEnabled=false
    Module.Param2WriteEnabled=false
    Module.Param3WriteEnabled=false
   ) {enabled=false, module=DynamicMaterialParameters001}
  ]
  particleUpdateScript [
   &Drag(Drag=0.9, Use Linear Drag=true, Use Rotational Drag=false)
   &InheritVelocity(
    Module.Limit Speed=true
    Velocity Limit=10000
    Velocity Scale=&MakeCustomVectorFromBool(Bool=$Particles.Initial.Locked, True Vector=<0 0 0>, False Vector=<-0.03 -0.03 -0.01>)
    Velocity Source=&TransformVector(Vector=$Engine.Owner.Velocity, Source Space=::World, Destination Space=::Local)
   ) {enabled=false}
   &AddVelocity(Velocity=&Multiply_VectorByFloat(
     Vector=<0 0 1>
     Float=&FloatFromCurve(FloatCurve=DataInterface() {propertyValues {Curve {keys [<at 0 0 RCIM_Cubic>, {arriveTangent=2.8829, interpMode=RCIM_Cubic, leaveTangent=2.8829, tangentMode=RCTM_User, time=1, value=1}]}}})
    )) {module=AddVelocity002}
   &SolveForcesAndVelocity(Clamp Velocity=false, Limit Acceleration=false)
   &FN_UpdateRibbon(
    Enable Skeletal Position=true
    Initial Color:LinearColor=$Particles.Initial.Color
    Initial RibbonWidth:NiagaraFloat=$Particles.Initial.RibbonWidth
    Sampled Position Coordinate Space:ENiagaraCoordinateSpace=::Local
    Sampled Rotation Coordinate Space:ENiagaraCoordinateSpace=::Local
    Sampled Velocity Coordinate Space:ENiagaraCoordinateSpace=::Local
   )
   &Material(Base Texture Time:NiagaraFloat=&Multiply_Float(A=$Emitter.Age, B=-1))
   &DynamicMaterialParameters(
    Index 0 Param 1=1
    Index 0 Param 2=1
    Index 0 Param 3=0.1
    Index 0 Param 4=0.2
    Index 1 Param 1=&Multiply_Float(A=-1.6, B=$Emitter.Age)
    Index 1 Param 2=0.15
    Index 1 Param 3=1.3
    Index 1 Param 4=1.8
    Index 2 Param 1=0.2
    Index 2 Param 2=0.2
    Index 2 Param 3=0.2
    Index 3 Param 1=1
    Index 3 Param 2=1
    Index 3 Param 3=1
    Index 3 Param 4=1
    Module.Param0WriteEnabled=true
    Module.Param0WriteEnabled1=true
    Module.Param0WriteEnabled2=true
    Module.Param0WriteEnabled3=true
    Module.Param1WriteEnabled=true
    Module.Param1WriteEnabled1=true
    Module.Param1WriteEnabled2=true
    Module.Param1WriteEnabled3=true
    Module.Param2WriteEnabled=true
    Module.Param2WriteEnabled1=true
    Module.Param2WriteEnabled2=true
    Module.Param2WriteEnabled3=true
    Module.Param3WriteEnabled=true
    Module.Param3WriteEnabled1=true
    Module.Param3WriteEnabled2=true
    Module.Param3WriteEnabled3=true
    Write Parameter Index 1=true
    Write Parameter Index 2=true
    Write Parameter Index 3=true
   ) {enabled=false}
   &ScaleRibbonWidth(Ribbon Width Scale=&FloatFromCurve(
     FloatCurve=DataInterface() {propertyValues {Curve {keys [
         {arriveTangent=-0.0063, interpMode=RCIM_Cubic, leaveTangent=-0.0063, tangentMode=RCTM_User, time=0.0058, value=1}
         {arriveTangent=1.4891, interpMode=RCIM_Cubic, leaveTangent=1.4891, time=0.4, value=1.4}
         <at 0.6774 2>
        ]}}}
     Scale Curve=0.6
    ))
   &ScaleColor(
    Scale Alpha=&Multiply_Float(
     A=&FloatFromCurve(FloatCurve=DataInterface() {propertyValues {Curve {keys [
          {arriveTangent=-9.7386, interpMode=RCIM_Cubic, leaveTangent=-9.7386, tangentMode=RCTM_User, time=0, value=4}
          {arriveTangent=-6.027, interpMode=RCIM_Cubic, leaveTangent=-6.027, tangentMode=RCTM_User, time=0.2748, value=1.6532}
          {arriveTangent=-0.639, interpMode=RCIM_Cubic, leaveTangent=-0.639, tangentMode=RCTM_User, time=1, value=0}
         ]}}})
     B=&Lerp_Float(A=0, B=1, Alpha=$System.Dive)
    )
    Scale RGB=&MakeVectorFromLinearColorRGB(Color=&ColorFromCurve(ColorCurve=DataInterface() {propertyValues {
        AlphaCurve {keys [<at 0 1 RCIM_Cubic>,<at 0.9344 0 RCIM_Cubic>]}
        BlueCurve {keys [<at 0 0 RCIM_Cubic>,<at 0.0392 0>,<at 0.1638 0.25>,<at 0.3615 0.1927>,<at 1 0.1667 RCIM_Cubic>]}
        GreenCurve {keys [<at 0 1.6766 RCIM_Cubic>,<at 0.0392 0.3353>,<at 0.1638 0.25>,<at 0.3615 0.1927>,<at 1 0.1667 RCIM_Cubic>]}
        RedCurve {keys [<at 0 5 RCIM_Cubic>,<at 0.0392 1>,<at 0.1638 0.25>,<at 0.3615 0.1927>,<at 1 0.1667 RCIM_Cubic>]}
       }}))
    ScaleA=true
    ScaleRGB=true
   )
  ]
  rendererClasses [&NiagaraRibbonRendererProperties]
  renderers [{rendererClass=&NiagaraRibbonRendererProperties}]
 }
 Syrup_Ribbons003 {
  emitterUpdateScript [
   &EmitterState()
   &FN_SpawnRibbon_v2(
    Number of Ribbons=2
    Spawn Mode=::NewEnumerator0
    SpawnRate:NiagaraFloat=&Lerp_Float(A=20, B=0, Alpha=$System.Dive)
   ) {module=FN_SpawnRibbon_v2001}
  ]
  particleSpawnScript [
   &FN_InitializeRibbon_v3(
    Enable Skeletal Position=true
    Lifetime=0.6
    Ribbon Width=30
    Sampled Position Coordinate Space:ENiagaraCoordinateSpace=::Local
    Sampled Rotation Coordinate Space:ENiagaraCoordinateSpace=::Local
    Sampled Velocity Coordinate Space:ENiagaraCoordinateSpace=::Local
   ) {module=FN_InitializeRibbon_v3001}
   &AddVelocity(
    Coordinate Space=::Simulation
    Velocity=&MakeCustomVectorFromBool(
     Bool=$Particles.Locked
     True Vector=<0 0 0>
     False Vector=&FN_FrontEndVector3Modifier(Gameplay=<0 0 200>, Front End=<0 0 350>)
    )
   ) {module=AddVelocity001}
   &DynamicMaterialParameters(
    Index 0 Param 1=&MaskFloatBySpawnGroup(Masked Float=0, Passthrough Float=0.5)
    Module.Param0WriteEnabled=true
    Module.Param1WriteEnabled=false
    Module.Param2WriteEnabled=false
    Module.Param3WriteEnabled=false
   ) {enabled=false, module=DynamicMaterialParameters001}
  ]
  particleUpdateScript [
   &Drag(Drag=0.9, Use Linear Drag=true, Use Rotational Drag=false)
   &InheritVelocity(
    Module.Limit Speed=true
    Velocity Limit=10000
    Velocity Scale=&MakeCustomVectorFromBool(Bool=$Particles.Initial.Locked, True Vector=<0 0 0>, False Vector=<-0.03 -0.03 -0.01>)
    Velocity Source=&TransformVector(Vector=$Engine.Owner.Velocity, Source Space=::World, Destination Space=::Local)
   ) {enabled=false}
   &AddVelocity(Velocity=&Multiply_VectorByFloat(
     Vector=<0 0 1>
     Float=&FloatFromCurve(FloatCurve=DataInterface() {propertyValues {Curve {keys [<at 0 0 RCIM_Cubic>, {arriveTangent=2.8829, interpMode=RCIM_Cubic, leaveTangent=2.8829, tangentMode=RCTM_User, time=1, value=1}]}}})
    )) {module=AddVelocity002}
   &SolveForcesAndVelocity(Clamp Velocity=false, Limit Acceleration=false)
   &FN_UpdateRibbon(
    Enable Skeletal Position=true
    Initial Color:LinearColor=$Particles.Initial.Color
    Initial RibbonWidth:NiagaraFloat=$Particles.Initial.RibbonWidth
    Sampled Position Coordinate Space:ENiagaraCoordinateSpace=::Local
    Sampled Rotation Coordinate Space:ENiagaraCoordinateSpace=::Local
    Sampled Velocity Coordinate Space:ENiagaraCoordinateSpace=::Local
   )
   &Material(Base Texture Time:NiagaraFloat=&Multiply_Float(A=$Emitter.Age, B=-1))
   &DynamicMaterialParameters(
    Index 0 Param 1=1
    Index 0 Param 2=1
    Index 0 Param 3=0.1
    Index 0 Param 4=0.2
    Index 1 Param 1=&Multiply_Float(A=-1.6, B=$Emitter.Age)
    Index 1 Param 2=0.15
    Index 1 Param 3=1.3
    Index 1 Param 4=1.8
    Index 2 Param 1=0.2
    Index 2 Param 2=0.2
    Index 2 Param 3=0.2
    Index 3 Param 1=1
    Index 3 Param 2=1
    Index 3 Param 3=1
    Index 3 Param 4=1
    Module.Param0WriteEnabled=true
    Module.Param0WriteEnabled1=true
    Module.Param0WriteEnabled2=true
    Module.Param0WriteEnabled3=true
    Module.Param1WriteEnabled=true
    Module.Param1WriteEnabled1=true
    Module.Param1WriteEnabled2=true
    Module.Param1WriteEnabled3=true
    Module.Param2WriteEnabled=true
    Module.Param2WriteEnabled1=true
    Module.Param2WriteEnabled2=true
    Module.Param2WriteEnabled3=true
    Module.Param3WriteEnabled=true
    Module.Param3WriteEnabled1=true
    Module.Param3WriteEnabled2=true
    Module.Param3WriteEnabled3=true
    Write Parameter Index 1=true
    Write Parameter Index 2=true
    Write Parameter Index 3=true
   ) {enabled=false}
   &ScaleRibbonWidth(Ribbon Width Scale=&FloatFromCurve(
     FloatCurve=DataInterface() {propertyValues {Curve {keys [
         {arriveTangent=-0.0063, interpMode=RCIM_Cubic, leaveTangent=-0.0063, tangentMode=RCTM_User, time=0.0058, value=1}
         {arriveTangent=1.4891, interpMode=RCIM_Cubic, leaveTangent=1.4891, time=0.4, value=1.4}
         <at 0.6774 2>
        ]}}}
     Scale Curve=0.6
    ))
   &ScaleColor(
    Scale Alpha=&Multiply_Float(
     A=&FloatFromCurve(FloatCurve=DataInterface() {propertyValues {Curve {keys [
          {arriveTangent=-10.3104, interpMode=RCIM_Cubic, leaveTangent=-10.3104, tangentMode=RCTM_User, time=0, value=2.5}
          {arriveTangent=-2.5201, interpMode=RCIM_Cubic, leaveTangent=-2.5201, tangentMode=RCTM_User, time=0.3, value=1}
          <at 1 0 RCIM_Cubic>
         ]}}})
     B=&Lerp_Float(A=1, B=0, Alpha=$System.Dive)
    )
    Scale RGB=&MakeVectorFromLinearColorRGB(Color=&ColorFromCurve(ColorCurve=DataInterface() {propertyValues {
        AlphaCurve {keys [<at 0 1 RCIM_Cubic>,<at 1 0 RCIM_Cubic>]}
        BlueCurve {keys [<at 0 0 RCIM_Cubic>,<at 0.0392 0>,<at 0.1638 0.1771>,<at 0.3615 0.1354>,<at 1 0.1667 RCIM_Cubic>]}
        GreenCurve {keys [<at 0 1.6766 RCIM_Cubic>,<at 0.0392 0.3353>,<at 0.1638 0.1771>,<at 0.3615 0.1354>,<at 1 0.1667 RCIM_Cubic>]}
        RedCurve {keys [<at 0 5 RCIM_Cubic>,<at 0.0392 1>,<at 0.1638 0.1771>,<at 0.3615 0.1354>,<at 1 0.1667 RCIM_Cubic>]}
       }}))
    ScaleA=true
    ScaleRGB=true
   )
  ]
  rendererClasses [&NiagaraRibbonRendererProperties]
  renderers [{rendererClass=&NiagaraRibbonRendererProperties}]
 }
 WingTrail_GP {
  emitterUpdateScript [
   &EmitterState(
    Inactive Response=::NewEnumerator1
    Life Cycle Mode=::NewEnumerator1
    Loop Behavior=::NewEnumerator1
    Loop Duration Mode=::NewEnumerator1
    UseLoopDelay=false
   )
   &SpawnBurst_Instantaneous(
    Spawn Count=&MakeCustomIntFromBool(Bool=$User.bIsFrontEndPreview, True INT=0, False INT=2)
    Use Spawn Probability=false
   )
  ]
  particleSpawnScript [
   &FN_SetSpawnGroup(Number of Spawn Groups=2)
   &InitializeParticle(
    Mesh Scale=&SelectVectorFromArray(
     Direct Array Index=$Particles.SpawnGroup
     Vector Selection Array=DataInterface() {propertyValues {FloatData [<1.6 1.6 1>, <1.6 1.6 1>, <0.8 0.8 3>, <0.8 0.8 0.5>]}}
    )
    Mesh Scale Mode=::NewEnumerator4
    UsePositionOffset=false
    Write Material Random=true
   )
   &InitialMeshOrientation(
    Orientation Axis=<0 0 1>
    Orientation Vector=<0 0 1>
    Rotation Coordinate Space=::NewEnumerator0
    Use Orientation Vector=true
    Use Rotation Vector=true
   )
   /SetVariables(Particles.WPOStrength=10) {isSetParameters=true, module=SetVariables_#2}
   &DynamicMaterialParameters(
    Index 0 Param 1=&SelectFloatFromArray(Direct Array Index=$Particles.SpawnGroup, Float Selection Array=DataInterface() {propertyValues {FloatData [0.25,0.2,0.3,0.4]}})
    Index 0 Param 2=&SelectFloatFromArray(Direct Array Index=$Particles.SpawnGroup, Float Selection Array=DataInterface() {propertyValues {FloatData [0,2.78,4.9,7.12]}})
    Index 0 Param 3=20
    Index 0 Param 4=1
    Index 2 Param 1=&SelectFloatFromArray(Direct Array Index=$Particles.SpawnGroup, Float Selection Array=DataInterface() {propertyValues {FloatData [0.78,1,0.85,0.93]}})
    Module.Param0WriteEnabled=true
    Module.Param0WriteEnabled2=true
    Module.Param1WriteEnabled=true
    Module.Param1WriteEnabled2=false
    Module.Param2WriteEnabled=false
    Module.Param2WriteEnabled2=false
    Module.Param3WriteEnabled=true
    Module.Param3WriteEnabled2=false
    Write Parameter Index 2=true
   )
   &CameraOffset(Camera Offset Amount=10)
   &Color(Color=&MultiplyLinearColor(
     Linear Color=<rgba 0.1354 0.1354 0.1354 0.5>
     Scale Factor=&MakeFloat4FromVectorAndFloat(
      "Vector (RGB)"=<1 1 1>
      "Float (Alpha)"=&SelectFloatFromArray(Direct Array Index=$Particles.SpawnGroup, Float Selection Array=DataInterface() {propertyValues {FloatData [1,1,0.7,0.7]}})
     )
    ))
  ]
  particleUpdateScript [
   &FN_DivingAttributes_v2(
    Accumulate Dive=0.1
    Accumulate Rest=-0.1
    Bone Index=2
    Sensitivity=1
    SkeletalMesh=DataInterface() {propertyValues {
      DefaultMesh=null
      ExcludeBoneName=root
      OptionalComponentClass=null
      PreviewMesh=&M_Med_Soldier_04_Preview
      SkinningMode=SkinOnTheFly
      SoftSourceActor=null
      SourceMode=Default
      UvSetIndex=0
      WholeMeshLOD=-1
      bExcludeBone=false
      bReadDeformedGeometry=true
      bRequireCurrentFrameData=true
      ComponentTags []
      FilteredBones [spine_02]
      FilteredSockets []
      MeshUserParameter {parameter {name=None, typeDefHandle {registeredTypeIndex=1}, varData []}}
      SamplingRegions []
     }}
   )
   &ParticleState(Kill Particles When Lifetime Has Elapsed=false, Let Infinitely Lived Particles Die When Emitter Deactivates=true)
   /SetVariables(Particles.Dive01:NiagaraFloat=&Multiply_Float(A=1, B=$Particles.Dive)) {isSetParameters=true, module=SetVariables_#3}
   &InterpolateOverTime(Rate Of Change=10, Target Vector Value:NiagaraPosition=$Particles.Position, Variable Type=::NewEnumerator2)
   /SetVariables(Particles.InterpolateDynamic:Vector4f=&MakeFloat4FromVectorAndFloat("Vector (RGB)"=$Output.InterpolateOverTime.Moving Average, "Float (Alpha)"=2)) {isSetParameters=true, module=SetVariables_#4}
   &SkeletalMeshLocation(
    Bone / Socket Index=$Particles.SpawnGroup
    Bone Sampling Mode=::NewEnumerator4
    Inherit Velocity=false
    MaskSpawnGroup=false
    Sampled Position Offset=<0 0 5>
    Skeletal Mesh=$User.SkeletalMesh
   )
   &ScaleMeshSize(Scale Factor=&Lerp_Vector(A=<0.6 0.6 0.6>, B=<0.6 0.6 1>, Alpha=$Particles.Dive))
   &DynamicMaterialParameters(
    Index 0 Param 3=&Multiply_Float(A=$Particles.WPOStrength, B=&Lerp_Float(A=0.3, B=0.5, Alpha=$Particles.Dive01))
    Index 2 Param 2=&Lerp_Float(A=100, B=100, Alpha=&Ceil_Float(Float=$Particles.Dive01))
    Index 2 Param 3=&Lerp_Float(A=0, B=0.5, Alpha=&Ceil_Float(Float=$Particles.Dive01))
    Index 2 Param 4=&Lerp_Float(A=0, B=0.5, Alpha=&Ceil_Float(Float=$Particles.Dive01))
    Module.Param0WriteEnabled=false
    Module.Param0WriteEnabled2=false
    Module.Param1WriteEnabled=false
    Module.Param1WriteEnabled2=true
    Module.Param2WriteEnabled=true
    Module.Param2WriteEnabled2=true
    Module.Param3WriteEnabled=false
    Module.Param3WriteEnabled2=true
    Write Parameter Index 2=true
   ) {module=DynamicMaterialParameters001}
   &ScaleColor(
    Scale Alpha=&Multiply_Float(
     A=&FNTimeOfDayScalar(Night Scale=0.5, Morning Scale=0.7, Evening Scale=0.7)
     B=&Lerp_Float(A=0, B=1, Alpha=$Particles.Dive)
    )
    ScaleA=true
    ScaleRGB=true
   ) {module=ScaleColor001}
   &CameraOffset(Camera Offset Amount=15) {module=CameraOffset001}
  ]
  rendererClasses [&NiagaraMeshRendererProperties]
  renderers [{rendererClass=&NiagaraMeshRendererProperties, rendererIndex=0}]
 }
 WingTrail_GP001 {
  emitterUpdateScript [
   &EmitterState(
    Inactive Response=::NewEnumerator1
    Life Cycle Mode=::NewEnumerator1
    Loop Behavior=::NewEnumerator1
    Loop Duration Mode=::NewEnumerator1
    UseLoopDelay=false
   )
   &SpawnBurst_Instantaneous(
    Spawn Count=&MakeCustomIntFromBool(Bool=$User.bIsFrontEndPreview, True INT=0, False INT=2)
    Use Spawn Probability=false
   )
  ]
  particleSpawnScript [
   &FN_SetSpawnGroup(Number of Spawn Groups=2)
   &InitializeParticle(
    Mesh Scale=&SelectVectorFromArray(
     Direct Array Index=$Particles.SpawnGroup
     Vector Selection Array=DataInterface() {propertyValues {FloatData [<6 6 1.5>, <6 6 1.5>, <0.8 0.8 0.5>, <0.8 0.8 0.5>]}}
    )
    Mesh Scale Mode=::NewEnumerator4
    UsePositionOffset=false
    Write Material Random=true
   )
   &InitialMeshOrientation(
    Orientation Axis=<0 0 1>
    Orientation Vector=<0 0 1>
    Rotation Coordinate Space=::NewEnumerator0
    Use Orientation Vector=true
    Use Rotation Vector=true
   )
   /SetVariables(Particles.WPOStrength=10) {isSetParameters=true, module=SetVariables_#2}
   &DynamicMaterialParameters(
    Index 0 Param 1=&SelectFloatFromArray(Direct Array Index=$Particles.SpawnGroup, Float Selection Array=DataInterface() {propertyValues {FloatData [0.25,0.2,0.3,0.4]}})
    Index 0 Param 2=&SelectFloatFromArray(Direct Array Index=$Particles.SpawnGroup, Float Selection Array=DataInterface() {propertyValues {FloatData [0,2.78,4.9,7.12]}})
    Index 0 Param 3=20
    Index 0 Param 4=1
    Index 2 Param 1=&SelectFloatFromArray(Direct Array Index=$Particles.SpawnGroup, Float Selection Array=DataInterface() {propertyValues {FloatData [0.78,1,0.85,0.93]}})
    Module.Param0WriteEnabled=true
    Module.Param0WriteEnabled2=true
    Module.Param1WriteEnabled=true
    Module.Param1WriteEnabled2=false
    Module.Param2WriteEnabled=false
    Module.Param2WriteEnabled2=false
    Module.Param3WriteEnabled=true
    Module.Param3WriteEnabled2=false
    Write Parameter Index 2=true
   )
   &CameraOffset(Camera Offset Amount=20)
   &Color(Color=&MultiplyLinearColor(
     Linear Color=<rgba 0.1406 0.1406 0.1406 0.8>
     Scale Factor=&MakeFloat4FromVectorAndFloat(
      "Vector (RGB)"=<1 1 1>
      "Float (Alpha)"=&SelectFloatFromArray(Direct Array Index=$Particles.SpawnGroup, Float Selection Array=DataInterface() {propertyValues {FloatData [2,2,0.7,0.7]}})
     )
    ))
  ]
  particleUpdateScript [
   &FN_DivingAttributes_v2(
    Accumulate Dive=0.1
    Accumulate Rest=-0.1
    Bone Index=0
    SkeletalMesh=DataInterface() {propertyValues {
      DefaultMesh=null
      ExcludeBoneName=root
      OptionalComponentClass=null
      PreviewMesh=&M_Med_Soldier_04_Preview
      SkinningMode=SkinOnTheFly
      SoftSourceActor=null
      SourceMode=Default
      UvSetIndex=0
      WholeMeshLOD=-1
      bExcludeBone=false
      bReadDeformedGeometry=true
      bRequireCurrentFrameData=true
      ComponentTags []
      FilteredBones [spine_02]
      FilteredSockets []
      MeshUserParameter {parameter {name=None, typeDefHandle {registeredTypeIndex=1}, varData []}}
      SamplingRegions []
     }}
   )
   &ParticleState(Kill Particles When Lifetime Has Elapsed=false, Let Infinitely Lived Particles Die When Emitter Deactivates=true)
   /SetVariables(Particles.Dive01:NiagaraFloat=&Multiply_Float(A=1, B=$Particles.Dive)) {isSetParameters=true, module=SetVariables_#3}
   &InterpolateOverTime(Rate Of Change=10, Target Vector Value:NiagaraPosition=$Particles.Position, Variable Type=::NewEnumerator2)
   /SetVariables(Particles.InterpolateDynamic:Vector4f=&MakeFloat4FromVectorAndFloat("Vector (RGB)"=$Output.InterpolateOverTime.Moving Average, "Float (Alpha)"=0.6)) {isSetParameters=true, module=SetVariables_#4}
   &SkeletalMeshLocation(
    Bone / Socket Index=$Particles.SpawnGroup
    Bone Sampling Mode=::NewEnumerator4
    Inherit Velocity=false
    MaskSpawnGroup=false
    Sampled Position Offset=<0 0 5>
    Skeletal Mesh=$User.SkeletalMesh
   )
   &ScaleColor(
    Scale Alpha=&Multiply_Float(
     A=&FNTimeOfDayScalar(Night Scale=0.5, Morning Scale=0.7, Evening Scale=0.7)
     B=&Lerp_Float(A=0, B=1, Alpha=$Particles.Dive)
    )
    ScaleA=true
    ScaleRGB=true
   ) {module=ScaleColor001}
   &ScaleMeshSize(Scale Factor=&Lerp_Vector(A=<0.5 0.5 0.6>, B=<0.5 0.5 1>, Alpha=$Particles.Dive))
   &DynamicMaterialParameters(
    Index 0 Param 3=&Multiply_Float(A=$Particles.WPOStrength, B=&Lerp_Float(A=0.3, B=0.48, Alpha=$Particles.Dive01))
    Index 2 Param 2=&Lerp_Float(A=100, B=100, Alpha=&Ceil_Float(Float=$Particles.Dive01))
    Index 2 Param 3=&Lerp_Float(A=0, B=0.5, Alpha=&Ceil_Float(Float=$Particles.Dive01))
    Index 2 Param 4=&Lerp_Float(A=0, B=0.5, Alpha=&Ceil_Float(Float=$Particles.Dive01))
    Module.Param0WriteEnabled=false
    Module.Param0WriteEnabled2=false
    Module.Param1WriteEnabled=false
    Module.Param1WriteEnabled2=true
    Module.Param2WriteEnabled=true
    Module.Param2WriteEnabled2=true
    Module.Param3WriteEnabled=false
    Module.Param3WriteEnabled2=true
    Write Parameter Index 2=true
   ) {module=DynamicMaterialParameters001}
  ]
  rendererClasses [&NiagaraMeshRendererProperties]
  renderers [{rendererClass=&NiagaraMeshRendererProperties, rendererIndex=0}]
 }
 WingTrail_GP002 {
  emitterUpdateScript [
   &EmitterState(
    Inactive Response=::NewEnumerator1
    Life Cycle Mode=::NewEnumerator1
    Loop Behavior=::NewEnumerator1
    Loop Duration Mode=::NewEnumerator1
    UseLoopDelay=false
   )
   &SpawnBurst_Instantaneous(
    Spawn Count=&MakeCustomIntFromBool(Bool=$User.bIsFrontEndPreview, True INT=0, False INT=2)
    Use Spawn Probability=false
   )
  ]
  particleSpawnScript [
   &FN_SetSpawnGroup(Number of Spawn Groups=2)
   &InitializeParticle(
    Mesh Scale=&SelectVectorFromArray(
     Direct Array Index=$Particles.SpawnGroup
     Vector Selection Array=DataInterface() {propertyValues {FloatData [<4 4 1.5>, <4 4 1.5>, <0.8 0.8 0.5>, <0.8 0.8 0.5>]}}
    )
    Mesh Scale Mode=::NewEnumerator4
    UsePositionOffset=false
    Write Material Random=true
   )
   &InitialMeshOrientation(
    Orientation Axis=<0 0 1>
    Orientation Vector=<0 0 1>
    Rotation Coordinate Space=::NewEnumerator0
    Use Orientation Vector=true
    Use Rotation Vector=true
   )
   /SetVariables(Particles.WPOStrength=10) {isSetParameters=true, module=SetVariables_#2}
   &DynamicMaterialParameters(
    Index 0 Param 1=&SelectFloatFromArray(Direct Array Index=$Particles.SpawnGroup, Float Selection Array=DataInterface() {propertyValues {FloatData [0.25,0.2,0.3,0.4]}})
    Index 0 Param 2=&SelectFloatFromArray(Direct Array Index=$Particles.SpawnGroup, Float Selection Array=DataInterface() {propertyValues {FloatData [0,2.78,4.9,7.12]}})
    Index 0 Param 3=20
    Index 0 Param 4=1
    Index 2 Param 1=&SelectFloatFromArray(Direct Array Index=$Particles.SpawnGroup, Float Selection Array=DataInterface() {propertyValues {FloatData [0.78,1,0.85,0.93]}})
    Module.Param0WriteEnabled=true
    Module.Param0WriteEnabled2=true
    Module.Param1WriteEnabled=true
    Module.Param1WriteEnabled2=false
    Module.Param2WriteEnabled=false
    Module.Param2WriteEnabled2=false
    Module.Param3WriteEnabled=true
    Module.Param3WriteEnabled2=false
    Write Parameter Index 2=true
   )
   &CameraOffset(Camera Offset Amount=20)
   &Color(Color=&MultiplyLinearColor(
     Linear Color=<rgba 0.1198 0.1198 0.1198 0.65>
     Scale Factor=&MakeFloat4FromVectorAndFloat(
      "Vector (RGB)"=<1 1 1>
      "Float (Alpha)"=&SelectFloatFromArray(Direct Array Index=$Particles.SpawnGroup, Float Selection Array=DataInterface() {propertyValues {FloatData [1,1,0.7,0.7]}})
     )
    ))
  ]
  particleUpdateScript [
   &FN_DivingAttributes_v2(
    Accumulate Dive=0.1
    Accumulate Rest=-0.1
    Bone Index=0
    SkeletalMesh=DataInterface() {propertyValues {
      DefaultMesh=null
      ExcludeBoneName=root
      OptionalComponentClass=null
      PreviewMesh=&M_Med_Soldier_04_Preview
      SkinningMode=SkinOnTheFly
      SoftSourceActor=null
      SourceMode=Default
      UvSetIndex=0
      WholeMeshLOD=-1
      bExcludeBone=false
      bReadDeformedGeometry=true
      bRequireCurrentFrameData=true
      ComponentTags []
      FilteredBones [spine_02]
      FilteredSockets []
      MeshUserParameter {parameter {name=None, typeDefHandle {registeredTypeIndex=1}, varData []}}
      SamplingRegions []
     }}
   )
   &ParticleState(Kill Particles When Lifetime Has Elapsed=false, Let Infinitely Lived Particles Die When Emitter Deactivates=true)
   /SetVariables(Particles.Dive01:NiagaraFloat=&Multiply_Float(A=1, B=$Particles.Dive)) {isSetParameters=true, module=SetVariables_#3}
   &InterpolateOverTime(Rate Of Change=10, Target Vector Value:NiagaraPosition=$Particles.Position, Variable Type=::NewEnumerator2)
   /SetVariables(Particles.InterpolateDynamic:Vector4f=&MakeFloat4FromVectorAndFloat("Vector (RGB)"=$Output.InterpolateOverTime.Moving Average, "Float (Alpha)"=0.6)) {isSetParameters=true, module=SetVariables_#4}
   &SkeletalMeshLocation(
    Bone / Socket Index=$Particles.SpawnGroup
    Bone Sampling Mode=::NewEnumerator4
    Inherit Velocity=false
    MaskSpawnGroup=false
    Sampled Position Offset=<0 0 5>
    Skeletal Mesh=$User.SkeletalMesh
   )
   &ScaleColor(
    Scale Alpha=&Multiply_Float(
     A=&FNTimeOfDayScalar(Night Scale=0.5, Morning Scale=0.7, Evening Scale=0.7)
     B=&Lerp_Float(A=0, B=1, Alpha=$Particles.Dive01)
    )
    ScaleA=true
    ScaleRGB=true
   ) {module=ScaleColor001}
   &ScaleMeshSize(Scale Factor=&Lerp_Vector(A=<0.5 0.5 0.6>, B=<0.5 0.5 1>, Alpha=$Particles.Dive01))
   &DynamicMaterialParameters(
    Index 0 Param 3=&Multiply_Float(A=$Particles.WPOStrength, B=&Lerp_Float(A=0.2, B=0.8, Alpha=$Particles.Dive01))
    Index 2 Param 2=&Lerp_Float(A=100, B=100, Alpha=&Ceil_Float(Float=$Particles.Dive01))
    Index 2 Param 3=&Lerp_Float(A=0, B=0.5, Alpha=&Ceil_Float(Float=$Particles.Dive01))
    Index 2 Param 4=&Lerp_Float(A=0, B=0.5, Alpha=&Ceil_Float(Float=$Particles.Dive01))
    Module.Param0WriteEnabled=false
    Module.Param0WriteEnabled1=true
    Module.Param0WriteEnabled2=false
    Module.Param1WriteEnabled=false
    Module.Param1WriteEnabled1=true
    Module.Param1WriteEnabled2=true
    Module.Param2WriteEnabled=true
    Module.Param2WriteEnabled1=true
    Module.Param2WriteEnabled2=true
    Module.Param3WriteEnabled=false
    Module.Param3WriteEnabled1=true
    Module.Param3WriteEnabled2=true
    Write Parameter Index 1=true
    Write Parameter Index 2=true
   ) {module=DynamicMaterialParameters001}
  ]
  rendererClasses [&NiagaraMeshRendererProperties]
  renderers [{rendererClass=&NiagaraMeshRendererProperties, rendererIndex=0}]
 }
}
summary {
 systemName=NS_Contrail_BoostWings
 emitters [
  {emitterName=Smoke, rendererClasses [&NiagaraSpriteRendererProperties]}
  {emitterName=WingTrail_GP, rendererClasses [&NiagaraMeshRendererProperties]}
  {emitterName=WingTrail_GP001, rendererClasses [&NiagaraMeshRendererProperties]}
  {emitterName=WingTrail_GP002, rendererClasses [&NiagaraMeshRendererProperties]}
  {emitterName=Light, rendererClasses [&NiagaraSpriteRendererProperties, &NiagaraLightRendererProperties]}
  {emitterName=Spark, rendererClasses [&NiagaraSpriteRendererProperties]}
  {emitterName=Flash, rendererClasses [&NiagaraMeshRendererProperties]}
  {emitterName=Flash001, rendererClasses [&NiagaraMeshRendererProperties]}
  {emitterName=Smoke1, rendererClasses [&NiagaraSpriteRendererProperties]}
  {emitterName=Syrup_Ribbons, rendererClasses [&NiagaraRibbonRendererProperties]}
  {emitterName=Spark1, rendererClasses [&NiagaraSpriteRendererProperties]}
  {emitterName=Smoke2, rendererClasses [&NiagaraSpriteRendererProperties]}
  {emitterName=Syrup_Ribbons001, rendererClasses [&NiagaraRibbonRendererProperties]}
  {emitterName=Syrup_Ribbons002, rendererClasses [&NiagaraRibbonRendererProperties]}
  {emitterName=Syrup_Ribbons003, rendererClasses [&NiagaraRibbonRendererProperties]}
  {emitterName=Mesh01, rendererClasses [&NiagaraMeshRendererProperties]}
  {emitterName=Mesh02, rendererClasses [&NiagaraMeshRendererProperties]}
 ]
 userVariables [
  User.bIsFrontEnd {type=NiagaraBool}
  User.bIsFrontEndPreview {type=NiagaraBool}
  User.Div {type=NiagaraFloat, defaultValue=0, description=""}
  User.RibbonIDOffset {type=NiagaraInt32}
  User.Skeletal Mesh01 {
   type=NiagaraDataInterfaceSkeletalMesh
   defaultValue {
    struct=NiagaraExt_VariableValue_DataInterface
    value {dataInterface=&self:NiagaraDataInterfaceSkeletalMesh_4}
   }
  }
  User.Skeletal Mesh02 {
   type=NiagaraDataInterfaceSkeletalMesh
   defaultValue {
    struct=NiagaraExt_VariableValue_DataInterface
    value {dataInterface=&self:NiagaraDataInterfaceSkeletalMesh_3}
   }
  }
  User.SkeletalMesh {
   type=NiagaraDataInterfaceSkeletalMesh
   defaultValue {
    struct=NiagaraExt_VariableValue_DataInterface
    value {dataInterface=&self:NiagaraDataInterfaceSkeletalMesh_2}
   }
  }
 ]
}
